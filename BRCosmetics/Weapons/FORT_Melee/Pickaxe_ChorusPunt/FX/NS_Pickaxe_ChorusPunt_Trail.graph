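# Niagara swing trail for the ChorusPunt pickaxe, built from two ribbon emitters using anim trail modules driven by bone or socket indices.
# MainTrail draws two ribbon renderers tinted by a dim warm color curve, with alpha following a system swing fade.
# TipTrail adds a ribbon on another socket pair that starts bright, then dims and narrows over its short life.
# System time-of-day values scale both trails' color, and dynamic material parameters feed the ribbon materials.
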NiagaraSystem /BRCosmetics/Weapons/FORT_Melee/Pickaxe_ChorusPunt/FX/NS_Pickaxe_ChorusPunt_Trail
dependencies {
 usedDataInterfaces [&NiagaraDataInterfaceColorCurve, &NiagaraDataInterfaceCurve, &NiagaraDataInterfaceSocketReader]
 usedDynamicInputs [&FNTimeOfDayScalar, &MergeHelper_FloatSwitch]
 usedModules [
  &self:SystemSpawnScript.SystemScriptSource.SystemScriptGraph.NiagaraNodeAssignment_1.SetVariables_#1
  &self:SystemSpawnScript.SystemScriptSource.SystemScriptGraph.NiagaraNodeAssignment_5.SetVariables_#2
  &self:SystemSpawnScript.SystemScriptSource.SystemScriptGraph.NiagaraNodeAssignment_0.SetVariables_#3
  &self:SystemSpawnScript.SystemScriptSource.SystemScriptGraph.NiagaraNodeAssignment_2.SetVariables_#4
  &SystemState
  &self:SystemSpawnScript.SystemScriptSource.SystemScriptGraph.NiagaraNodeAssignment_4.SetVariables_#5
 ]
}
emitters {
 MainTrail {
  emitterSpawnScript [&FN_RibbonIDOffsetUtility()]
  emitterUpdateScript [
   &EmitterState()
   &FN_SpawnAnimTrail()
  ]
  particleSpawnScript [
   &FN_InitializeAnimTrailBoneIndecies()
   &FN_InitializeAnimTrail(AnimTrailWidth=1.04, Color=$System.TOD, Lifetime=0.3)
  ]
  particleUpdateScript [
   &FN_UpdateAnimTrail() {module=FN_UpdateAnimTrail001}
   &ScaleColor(
    Scale Alpha=$System.SwingGeneralFade
    Scale RGB=&VectorFromFloat(Value=$System.ToDScalar)
    ScaleA=true
    ScaleRGB=true
   ) {module=ScaleColor001}
   &DynamicMaterialParameters(
    Index 0 Param 1=1
    Index 0 Param 3=1
    Index 0 Param 4=1
    Index 1 Param 3=0.1
    Index 1 Param 4=1
    Index 2 Param 1=10
    Index 2 Param 2=3
    Index 2 Param 4=0.6
    Module.Index0Float0Write=true
    Module.Index0Float1Write=true
    Module.Index0Float2Write=true
    Module.Index0Float3Write=true
    Module.Index1Float0Write=true
    Module.Index1Float1Write=true
    Module.Index1Float2Write=true
    Module.Index1Float3Write=true
    Module.Index2Float0Write=true
    Module.Index2Float1Write=true
    Module.Index2Float2Write=false
    Module.Index2Float3Write=false
    Write Parameter Index 1=::NewEnumerator1
    Write Parameter Index 2=::NewEnumerator1
   )
   /SetVariables(
    Particles.DynamicParametersAlt:Vector4f=&MakeVector4(X=0, Y=0, Z=0.1, W=1)
    Particles.DynamicParametersAlt2:Vector4f=&MakeVector4(X=1, Y=0, Z=1, W=1)
   ) {isSetParameters=true}
   &ScaleColor(
    Scale Alpha=0.6
    Scale RGB=&Multiply_Vector(
     A=&MakeVectorFromLinearColorRGB(Color=&ColorFromCurve(ColorCurve=DataInterface() {propertyValues {
         AlphaCurve {keys []}
         BlueCurve {keys [<at 0 0.013 RCIM_Cubic>,<at 0.5683 0.0091 RCIM_Cubic>]}
         GreenCurve {keys [<at 0 0.162 RCIM_Cubic>,<at 0.5683 0.1144 RCIM_Cubic>]}
         RedCurve {keys [<at 0 0.4179 RCIM_Cubic>,<at 0.5683 0.3467 RCIM_Cubic>]}
        }}))
     B=<1 1 1>
    )
    ScaleA=true
    ScaleRGB=true
   )
  ]
  rendererClasses [&NiagaraRibbonRendererProperties]
  renderers [
   {rendererClass=&NiagaraRibbonRendererProperties}
   {rendererClass=&NiagaraRibbonRendererProperties, rendererIndex=1}
  ]
 }
 TipTrail {
  emitterSpawnScript [&FN_RibbonIDOffsetUtility()]
  emitterUpdateScript [
   &EmitterState()
   &FN_SpawnAnimTrail()
  ]
  particleSpawnScript [
   &FN_InitializeAnimTrailBoneIndecies(First Bone or Socket Index=1, Second Bone or Socket Index=2)
   &FN_InitializeAnimTrail(AnimTrailWidth=1.2, Color=<rgba 0.3565 0.1366 0.0093 1>, Lifetime=0.3)
  ]
  particleUpdateScript [
   &FN_UpdateAnimTrail() {module=FN_UpdateAnimTrail001}
   &ScaleColor(
    Scale Alpha=$System.SwingGeneralFade
    Scale RGB=&VectorFromFloat(Value=&FloatFromCurve(FloatCurve=DataInterface() {propertyValues {Curve {keys [<at 0 3>,<at 0.2 1>,<at 1 0>]}}}, Scale Curve=$System.ToDScalar))
    ScaleA=true
    ScaleRGB=true
   ) {module=ScaleColor001}
   &ScaleRibbonWidth(Ribbon Width Scale=&FloatFromCurve(FloatCurve=DataInterface() {propertyValues {Curve {keys [<at 0 1>, {leaveTangent=-1.7704, time=1, value=0}]}}}))
   &DynamicMaterialParameters(
    Index 0 Param 1=1
    Index 0 Param 3=1
    Index 0 Param 4=1
    Index 1 Param 2=0.3
    Index 1 Param 3=0.1
    Index 1 Param 4=1
    Index 2 Param 1=20
    Index 2 Param 2=15
    Index 2 Param 4=0.6
    Module.Index0Float0Write=true
    Module.Index0Float1Write=true
    Module.Index0Float2Write=true
    Module.Index0Float3Write=true
    Module.Index1Float0Write=true
    Module.Index1Float1Write=true
    Module.Index1Float2Write=true
    Module.Index1Float3Write=true
    Module.Index2Float0Write=true
    Module.Index2Float1Write=true
    Module.Index2Float2Write=false
    Module.Index2Float3Write=false
    Write Parameter Index 1=::NewEnumerator1
    Write Parameter Index 2=::NewEnumerator1
   )
   &ScaleColor(Scale RGB=&MakeVectorFromLinearColorRGB(Color=$System.TOD), ScaleA=true, ScaleRGB=true)
  ]
  rendererClasses [&NiagaraRibbonRendererProperties]
  renderers [{rendererClass=&NiagaraRibbonRendererProperties}]
 }
}
summary {
 systemName=NS_Pickaxe_ChorusPunt_Trail
 emitters [
  {emitterName=MainTrail, rendererClasses [&NiagaraRibbonRendererProperties]}
  {emitterName=TipTrail, rendererClasses [&NiagaraRibbonRendererProperties]}
 ]
 userVariables [
  User.RibbonIDOffset {type=NiagaraInt32}
  User.SwingSectionIndex {type=NiagaraInt32}
 ]
}
